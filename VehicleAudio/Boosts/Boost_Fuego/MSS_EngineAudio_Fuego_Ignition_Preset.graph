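# Mono one-shot MetaSound Source preset that supplies the engine ignition sound for the Fuego vehicle boost cosmetic.
# It inherits its graph from MSS_EngineAudio_Ignition, overriding inputs for the ignition wave, volumes and exhaust-pop delay, count and repeat rate.
# The exhaust-pop wave array is empty, leaving only the Fuego ignition wave assigned.
# Vehicle attenuation, ignition concurrency and engine volume, high-pass and low-pass modulation shape playback.
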MetaSoundSource /VehicleAudio/Boosts/Boost_Fuego/MSS_EngineAudio_Fuego_Ignition_Preset
preset_parent=/VehicleAudio/Patches/MSS_EngineAudio_Ignition
inputs [
 ExhaustPop_StartDelayTime {type=Time, access=Constructor}
 ExhaustPopPitch {type=Float}
 ExhaustPops {type=WaveAsset:Array, access=Constructor}
 ExhaustPopVolume {type=Float, access=Constructor}
 IgnitionSound {type=WaveAsset:Array, access=Constructor}
 IgnitionVolume {type=Float, access=Constructor}
 MaxNumExhaustPopsToPlay {type=Int32, access=Constructor}
 MaxRepeatRate {type=Float, access=Constructor}
 MinNumExhaustPopsToPlay {type=Int32, access=Constructor}
 MinRepeatRate {type=Float, access=Constructor}
 UE.Source.OnPlay {type=Trigger}
]
interfaces [UE.OutputFormat.Mono:1.0, UE.Source.OneShot:1.0, UE.Source:1.1]
outputs [
 "UE.OutputFormat.Mono.Audio:0" {type=Audio}
 UE.Source.OneShot.OnFinished {type=Trigger}
]
pages [Default {input_defaults {
   ExhaustPopPitch=0
   ExhaustPopVolume=-2
   ExhaustPop_StartDelayTime=0.8
   IgnitionVolume=-1
   MaxNumExhaustPopsToPlay=10
   MaxRepeatRate=0.2
   MinNumExhaustPopsToPlay=4
   MinRepeatRate=0.1
   UE.Source.OnPlay=null
   ExhaustPops []
   IgnitionSound [&/Fuego_REV_Ignition_Dealership]
  }}]
properties {
 assetImportData=&self:AssetImportData
 attenuationSettings=/VehicleAudio/Settings/Vehicle_Movement_Att
 audioPropertiesSheet=/VehicleAudio/PropertySheets/APS_Cosmetic_Vehicle_Engine_Ignition
 maxDistance=12512
 sampleRate=48000
 soundClassObject=&Fort_Vehicles_Interior
 soundSubmixObject=&Fortnite_SFX
 concurrencySet [/VehicleAudio/Settings/Vehicle_Cosmetic_Ignition_Conc]
 modulationSettings {
  highpassRouting=Union
  lowpassRouting=Union
  volumeRouting=Union
  highpassModulationDestination {bEnableModulation=true, modulators [&PP_FortVehicles_Engine_HPF]}
  lowpassModulationDestination {bEnableModulation=true, modulators [&Patch_FortVehicles_Engine_LPF]}
  volumeModulationDestination {
   bEnableModulation=true
   modulators [&Patch_FortVehicles_Engine_Vol, &DM_VehicleEnginePatch]
  }
 }
}
